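; Gameplay cue for the Big Bush (leaf) grenade explosion, adjusting burst VFX and playing the explosion sound.
; On burst, the first particle component takes its Radius parameter from the cue's raw magnitude.
; A SurfaceType8 physical material counts as water, which overrides the normal-aligned effect rotation with a fixed one and selects the water explosion sound.
; The sound plays at the target actor's location.
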(fn OnBurst(EitherVFX-preferthisversion)
  (params (MyTarget :type Actor%20Object%20Reference) (Parameters :type Gameplay%20Cue%20Parameters%20Structure%20%28by%20ref%29) (ParticleComponents :type Array%20of%20FXSystem%20Component%20Object%20References) (AudioComponents :type Array%20of%20Audio%20Component%20Object%20References) (BurstCameraShakeInstance :type Camera%20Shake%20Base%20Object%20Reference) (BurstDecalInstance :type Decal%20Actor%20Object%20Reference))
  (bind (_normalizedmagnitude _rawmagnitude _effectcontext _matchedtagname _originaltag _aggregatedsourcetags _aggregatedtargettags _location _normal _instigator _effectcauser _sourceobject _physicalmaterial) (Ability|GameplayCue|BreakGameplayCueParameters Parameters))
  (bind _output (Utilities|Array|Get(acopy) ParticleComponents))
  (bind _iswater (LocalVariables|GetisWater))
  (Utilities|IsValid _physicalmaterial
    (:Is%20Valid
      (LocalVariables|SetisWater (Utilities|Enum|Equal(Enum) (Variables|/Script/PhysicsCore.PhysicalMaterial|GetSurfaceType _physicalmaterial) "SurfaceType8")))
    (:Is%20Not%20Valid
      _))
  (Effects|Components|ParticleSystem|SetFloatParameter _output "Radius" _rawmagnitude)
  (Transformation|SetWorldRotation _output (Math|Rotator|SelectRotator "90.000000, 0, 0" (Math|Conversions|RotationFromXVector _normal) _iswater) false false) ; Override orientation for water impacts
  (Audio|PlaySoundatLocation MyTarget (select _iswater "/BigBushGrenade/Items/Audio/BigBushGrenade_Explode_Water_Cue.BigBushGrenade_Explode_Water_Cue" "/BigBushGrenade/Items/Audio/SC_BigBushGrenade_Explode_Cue.SC_BigBushGrenade_Explode_Cue") (Transformation|GetActorLocation MyTarget))
  (return))
